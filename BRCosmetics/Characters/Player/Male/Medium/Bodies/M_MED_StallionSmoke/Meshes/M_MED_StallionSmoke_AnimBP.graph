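# Post-process-style AnimBlueprint for the StallionSmoke male medium cosmetic body, adding secondary motion to zippers and collar over the attached parent's copied pose.
# The pose runs through the Despair deform Control Rig, then weighted Constraints place hip, leg and arm zipper roots, each followed by spring AnimDynamics.
# Constrained collar bones get additive offsets, some fixed, others blended by head-versus-spine_05 rotation alphas computed each update.
# The computed neckBackRot is unused.
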AnimBlueprint /BRCosmetics/Characters/Player/Male/Medium/Bodies/M_MED_StallionSmoke/Meshes/M_MED_StallionSmoke_AnimBP
target_skeleton=&/M_MED_StallionSmoke_Skeleton
graphs [
 AnimGraph {nodes [
   "Transform (Modify) Bone - Bone: dyn_collar_01_c"(Alpha=Get neckDownRot() -> (neckDownRot:double), Translation="0,0,0", Rotation="20,0,0", Scale="1,1,1") {
    rotationMode=BMM_Additive
    rotationSpace=BCS_BoneSpace
    scaleMode=BMM_Additive
    scaleSpace=BCS_BoneSpace
    translationMode=BMM_Additive
    translationSpace=BCS_BoneSpace
    alphaBoolBlend {blendOption=Linear}
    alphaScaleBiasClamp {interpSpeedDecreasing=1}
    boneToModify {boneName=dyn_collar_01_c}
    rotation {pitch=20}
   }
   n1 = Control Rig(Source=Copy Pose From Mesh() {bUseAttachedParent=true}) {
    alphaBoolBlend {blendOption=Linear}
    controlRigAssetReference {blueprintRigClass=/BRCosmetics/Characters/Player/Male/Medium/Bodies/M_MED_Despair/Meshes/M_MED_Despair_DeformRig_Master_ControlRig_new.M_MED_Despair_DeformRig_Master_ControlRig_new_C}
   }
   n2 = "Constraint - dyn_hipZip_00_l_MSTR (PPPPPP)"!(
    ComponentPose=Local To Component(LocalPose=@n1)
    Alpha=1
    ConstraintWeights_0=0.03
    ConstraintWeights_1:float=0.03
    ConstraintWeights_2:float=0.77
    ConstraintWeights_3:float=0.07
    ConstraintWeights_4:float=0.01
    ConstraintWeights_5:float=0.09
   ) -> (Pose) {
    alphaBoolBlend {blendOption=Linear}
    boneToModify {boneName=dyn_hipZip_00_l_MSTR}
    constraintSetup [
     {targetBone {boneName=spine_01}}
     {targetBone {boneName=spine_02}}
     {targetBone {boneName=pelvis}}
     {targetBone {boneName=thigh_twist_01_l}}
     {targetBone {boneName=spine_03}}
     {targetBone {boneName=deform_groin_l}}
    ]
    constraintWeights [1,1,1,1,1,1]
   }
   n3 = Anim Dynamics - Bone: dyn_hipZip_01_l(ComponentPose=@n2, Alpha=1, ExternalForce="0,0,0") {
    angularDampingOverride=0.97
    angularSpringConstant=20000
    bAngularSpring=true
    bOverrideAngularDamping=true
    bUseGravityOverride=true
    simulationSpace=RootRelative
    alphaBoolBlend {blendOption=Linear}
    boundBone {boneName=dyn_hipZip_01_l}
    componentAppliedLinearAccClamp=<2000 2000 2000>
    componentLinearAccScale=<0.1 0.1 0.2>
    componentLinearVelScale=<0.2 0.2 0.4>
    gravityOverride {z=-980}
    physicsBodyDefinitions [{
      boundBone {boneName=dyn_hipZip_01_l}
      boxExtents=<10 20 20>
      constraintSetup {angularLimitsMax=<15 0 15>, angularLimitsMin=<-15 -45 -15>, angularTarget=<0 0 0>}
      localJointOffset=<4 0 0>
     }]
   }
   n4 = "Constraint - dyn_hipZip_00_r_MSTR (PPPPPP)"!(
    ComponentPose=@n3
    Alpha=1
    ConstraintWeights_0=0.03
    ConstraintWeights_1:float=0.03
    ConstraintWeights_2:float=0.77
    ConstraintWeights_3:float=0.01
    ConstraintWeights_4:float=0.07
    ConstraintWeights_5:float=0.09
   ) -> (Pose) {
    alphaBoolBlend {blendOption=Linear}
    boneToModify {boneName=dyn_hipZip_00_r_MSTR}
    constraintSetup [
     {targetBone {boneName=spine_01}}
     {targetBone {boneName=spine_02}}
     {targetBone {boneName=pelvis}}
     {targetBone {boneName=spine_03}}
     {targetBone {boneName=thigh_twist_01_r}}
     {targetBone {boneName=deform_groin_r}}
    ]
    constraintWeights [1,1,1,1,1,1]
   }
   n5 = Anim Dynamics - Bone: dyn_hipZip_01_r(ComponentPose=@n4, Alpha=1, ExternalForce="0,0,0") {
    angularDampingOverride=0.97
    angularSpringConstant=20000
    bAngularSpring=true
    bOverrideAngularDamping=true
    bUseGravityOverride=true
    simulationSpace=RootRelative
    alphaBoolBlend {blendOption=Linear}
    boundBone {boneName=dyn_hipZip_01_r}
    componentAppliedLinearAccClamp=<2000 2000 2000>
    componentLinearAccScale=<0.1 0.1 0.2>
    componentLinearVelScale=<0.2 0.2 0.4>
    gravityOverride {z=-980}
    physicsBodyDefinitions [{
      boundBone {boneName=dyn_hipZip_01_r}
      boxExtents=<10 20 20>
      constraintSetup {angularLimitsMax=<15 0 15>, angularLimitsMin=<-15 -45 -15>, angularTarget=<0 0 0>}
      localJointOffset=<4 0 0>
     }]
   }
   n6 = "Constraint - dyn_legZip_00_r_MSTR (PPPP)"!(
    ComponentPose=@n5
    Alpha=1
    ConstraintWeights_0=0.6
    ConstraintWeights_1:float=0.36
    ConstraintWeights_2:float=0.02
    ConstraintWeights_3:float=0.02
   ) -> (Pose) {
    alphaBoolBlend {blendOption=Linear}
    boneToModify {boneName=dyn_legZip_00_r_MSTR}
    constraintSetup [
     {targetBone {boneName=thigh_r}}
     {targetBone {boneName=thigh_twist_01_r}}
     {targetBone {boneName=calf_r}}
     {targetBone {boneName=deform_knee_out_r}}
    ]
    constraintWeights [1,1,1,1]
   }
   n7 = Anim Dynamics - Bone: dyn_legZip_01_r(ComponentPose=@n6, Alpha=1, ExternalForce="0,0,0") {
    angularDampingOverride=0.97
    angularSpringConstant=20000
    bAngularSpring=true
    bOverrideAngularDamping=true
    bUseGravityOverride=true
    simulationSpace=RootRelative
    alphaBoolBlend {blendOption=Linear}
    boundBone {boneName=dyn_legZip_01_r}
    componentAppliedLinearAccClamp=<2000 2000 2000>
    componentLinearAccScale=<0.1 0.1 0.2>
    componentLinearVelScale=<0.2 0.2 0.4>
    gravityOverride {z=-980}
    physicsBodyDefinitions [{
      boundBone {boneName=dyn_legZip_01_r}
      boxExtents=<10 20 20>
      constraintSetup {angularLimitsMax=<15 0 15>, angularLimitsMin=<-15 -45 -15>, angularTarget=<0 0 0>}
      localJointOffset=<4 0 0>
     }]
   }
   n8 = "Constraint - dyn_legZipA_00_l_MSTR (PPPP)"!(
    ComponentPose=@n7
    Alpha=1
    ConstraintWeights_0=0.08
    ConstraintWeights_1:float=0.06
    ConstraintWeights_2:float=0.78
    ConstraintWeights_3:float=0.08
   ) -> (Pose) {
    alphaBoolBlend {blendOption=Linear}
    boneToModify {boneName=dyn_legZipA_00_l_MSTR}
    constraintSetup [
     {targetBone {boneName=pelvis}}
     {targetBone {boneName=deform_groin_l}}
     {targetBone {boneName=thigh_twist_01_l}}
     {targetBone {boneName=thigh_l}}
    ]
    constraintWeights [1,1,1,1]
   }
   n9 = Anim Dynamics - Bone: dyn_legZipA_01_l(ComponentPose=@n8, Alpha=1, ExternalForce="0,0,0") {
    angularDampingOverride=0.97
    angularSpringConstant=20000
    bAngularSpring=true
    bOverrideAngularDamping=true
    bUseGravityOverride=true
    simulationSpace=RootRelative
    alphaBoolBlend {blendOption=Linear}
    boundBone {boneName=dyn_legZipA_01_l}
    componentAppliedLinearAccClamp=<2000 2000 2000>
    componentLinearAccScale=<0.1 0.1 0.2>
    componentLinearVelScale=<0.2 0.2 0.4>
    gravityOverride {z=-980}
    physicsBodyDefinitions [{
      boundBone {boneName=dyn_legZipA_01_l}
      boxExtents=<10 20 20>
      constraintSetup {angularLimitsMax=<15 0 15>, angularLimitsMin=<-15 -45 -15>, angularTarget=<0 0 0>}
      localJointOffset=<4 0 0>
     }]
   }
   n10 = "Constraint - dyn_legZipB_00_l_MSTR (PPPP)"!(
    ComponentPose=@n9
    Alpha=1
    ConstraintWeights_0=0.01
    ConstraintWeights_1:float=0.02
    ConstraintWeights_2:float=0.55
    ConstraintWeights_3:float=0.42
   ) -> (Pose) {
    alphaBoolBlend {blendOption=Linear}
    boneToModify {boneName=dyn_legZipB_00_l_MSTR}
    constraintSetup [
     {targetBone {boneName=deform_knee_out_l}}
     {targetBone {boneName=deform_groin_l}}
     {targetBone {boneName=thigh_twist_01_l}}
     {targetBone {boneName=thigh_l}}
    ]
    constraintWeights [1,1,1,1]
   }
   n11 = Anim Dynamics - Bone: dyn_legZipB_01_l(ComponentPose=@n10, Alpha=1, ExternalForce="0,0,0") {
    angularDampingOverride=0.97
    angularSpringConstant=20000
    bAngularSpring=true
    bOverrideAngularDamping=true
    bUseGravityOverride=true
    simulationSpace=RootRelative
    alphaBoolBlend {blendOption=Linear}
    boundBone {boneName=dyn_legZipB_01_l}
    componentAppliedLinearAccClamp=<2000 2000 2000>
    componentLinearAccScale=<0.1 0.1 0.2>
    componentLinearVelScale=<0.2 0.2 0.4>
    gravityOverride {z=-980}
    physicsBodyDefinitions [{
      boundBone {boneName=dyn_legZipB_01_l}
      boxExtents=<10 20 20>
      constraintSetup {angularLimitsMax=<15 0 15>, angularLimitsMin=<-15 -45 -15>, angularTarget=<0 0 0>}
      localJointOffset=<4 0 0>
     }]
   }
   n12 = "Constraint - dyn_legZipC_00_l_MSTR (PPPP)"!(
    ComponentPose=@n11
    Alpha=1
    ConstraintWeights_0=0.02
    ConstraintWeights_1:float=0.02
    ConstraintWeights_2:float=0.36
    ConstraintWeights_3:float=0.6
   ) -> (Pose) {
    alphaBoolBlend {blendOption=Linear}
    boneToModify {boneName=dyn_legZipC_00_l_MSTR}
    constraintSetup [
     {targetBone {boneName=deform_knee_out_l}}
     {targetBone {boneName=calf_l}}
     {targetBone {boneName=thigh_twist_01_l}}
     {targetBone {boneName=thigh_l}}
    ]
    constraintWeights [1,1,1,1]
   }
   n13 = Anim Dynamics - Bone: dyn_legZipC_01_l(ComponentPose=@n12, Alpha=1, ExternalForce="0,0,0") {
    angularDampingOverride=0.97
    angularSpringConstant=20000
    bAngularSpring=true
    bOverrideAngularDamping=true
    bUseGravityOverride=true
    simulationSpace=RootRelative
    alphaBoolBlend {blendOption=Linear}
    boundBone {boneName=dyn_legZipC_01_l}
    componentAppliedLinearAccClamp=<2000 2000 2000>
    componentLinearAccScale=<0.1 0.1 0.2>
    componentLinearVelScale=<0.2 0.2 0.4>
    gravityOverride {z=-980}
    physicsBodyDefinitions [{
      boundBone {boneName=dyn_legZipC_01_l}
      boxExtents=<10 20 20>
      constraintSetup {angularLimitsMax=<15 0 15>, angularLimitsMin=<-15 -45 -15>, angularTarget=<0 0 0>}
      localJointOffset=<4 0 0>
     }]
   }
   n14 = "Constraint - dyn_armZipA_00_l_MSTR (PPPPPP)"!(
    ComponentPose=@n13
    Alpha=1
    ConstraintWeights_0=0.43
    ConstraintWeights_1:float=0.07
    ConstraintWeights_2:float=0.08
    ConstraintWeights_3:float=0.24
    ConstraintWeights_4:float=0.14
    ConstraintWeights_5:float=0.04
   ) -> (Pose) {
    alphaBoolBlend {blendOption=Linear}
    boneToModify {boneName=dyn_armZipA_00_l_MSTR}
    constraintSetup [
     {targetBone {boneName=upperarm_twist_01_l}}
     {targetBone {boneName=clavicle_l}}
     {targetBone {boneName=upperarm_l}}
     {targetBone {boneName=upperarm_twist_02_l}}
     {targetBone {boneName=deform_upperarm_fr_l}}
     {targetBone {boneName=deform_upperarm_up_l}}
    ]
    constraintWeights [1,1,1,1,1,1]
   }
   n15 = Anim Dynamics - Bone: dyn_armZipA_01_l(ComponentPose=@n14, Alpha=1, ExternalForce="0,0,0") {
    angularDampingOverride=0.97
    angularSpringConstant=20000
    bAngularSpring=true
    bOverrideAngularDamping=true
    bUseGravityOverride=true
    simulationSpace=RootRelative
    alphaBoolBlend {blendOption=Linear}
    boundBone {boneName=dyn_armZipA_01_l}
    componentAppliedLinearAccClamp=<2000 2000 2000>
    componentLinearAccScale=<0.1 0.1 0.2>
    componentLinearVelScale=<0.2 0.2 0.4>
    gravityOverride {z=-980}
    physicsBodyDefinitions [{
      boundBone {boneName=dyn_armZipA_01_l}
      boxExtents=<10 20 20>
      constraintSetup {angularLimitsMax=<15 0 15>, angularLimitsMin=<-15 -45 -15>, angularTarget=<0 0 0>}
      localJointOffset=<4 0 0>
     }]
   }
   n16 = "Constraint - dyn_armZipB_00_l_MSTR (PPPPP)"!(
    ComponentPose=@n15
    Alpha=1
    ConstraintWeights_0=0.37
    ConstraintWeights_1:float=0.02
    ConstraintWeights_2:float=0.21
    ConstraintWeights_3:float=0.32
    ConstraintWeights_4:float=0.08
   ) -> (Pose) {
    alphaBoolBlend {blendOption=Linear}
    boneToModify {boneName=dyn_armZipB_00_l_MSTR}
    constraintSetup [
     {targetBone {boneName=upperarm_twist_01_l}}
     {targetBone {boneName=clavicle_l}}
     {targetBone {boneName=upperarm_l}}
     {targetBone {boneName=upperarm_twist_02_l}}
     {targetBone {boneName=deform_upperarm_fr_l}}
    ]
    constraintWeights [1,1,1,1,1]
   }
   n17 = Anim Dynamics - Bone: dyn_armZipB_01_l(ComponentPose=@n16, Alpha=1, ExternalForce="0,0,0") {
    angularDampingOverride=0.97
    angularSpringConstant=20000
    bAngularSpring=true
    bOverrideAngularDamping=true
    bUseGravityOverride=true
    simulationSpace=RootRelative
    alphaBoolBlend {blendOption=Linear}
    boundBone {boneName=dyn_armZipB_01_l}
    componentAppliedLinearAccClamp=<2000 2000 2000>
    componentLinearAccScale=<0.1 0.1 0.2>
    componentLinearVelScale=<0.2 0.2 0.4>
    gravityOverride {z=-980}
    physicsBodyDefinitions [{
      boundBone {boneName=dyn_armZipB_01_l}
      boxExtents=<10 20 20>
      constraintSetup {angularLimitsMax=<15 0 15>, angularLimitsMin=<-15 -45 -15>, angularTarget=<0 0 0>}
      localJointOffset=<4 0 0>
     }]
   }
   n18 = "Constraint - dyn_armZipC_00_l_MSTR (PPPP)"!(
    ComponentPose=@n17
    Alpha=1
    ConstraintWeights_0=0.36
    ConstraintWeights_1:float=0.32
    ConstraintWeights_2:float=0.05
    ConstraintWeights_3:float=0.27
   ) -> (Pose) {
    alphaBoolBlend {blendOption=Linear}
    boneToModify {boneName=dyn_armZipC_00_l_MSTR}
    constraintSetup [
     {targetBone {boneName=upperarm_twist_01_l}}
     {targetBone {boneName=upperarm_twist_02_l}}
     {targetBone {boneName=deform_upperarm_bk_l}}
     {targetBone {boneName=upperarm_l}}
    ]
    constraintWeights [1,1,1,1]
   }
   n19 = Anim Dynamics - Bone: dyn_armZipC_01_l(ComponentPose=@n18, Alpha=1, ExternalForce="0,0,0") {
    angularDampingOverride=0.97
    angularSpringConstant=20000
    bAngularSpring=true
    bOverrideAngularDamping=true
    bUseGravityOverride=true
    simulationSpace=RootRelative
    alphaBoolBlend {blendOption=Linear}
    boundBone {boneName=dyn_armZipC_01_l}
    componentAppliedLinearAccClamp=<2000 2000 2000>
    componentLinearAccScale=<0.1 0.1 0.2>
    componentLinearVelScale=<0.2 0.2 0.4>
    gravityOverride {z=-980}
    physicsBodyDefinitions [{
      boundBone {boneName=dyn_armZipC_01_l}
      boxExtents=<10 20 20>
      constraintSetup {angularLimitsMax=<15 0 15>, angularLimitsMin=<-15 -45 -15>, angularTarget=<0 0 0>}
      localJointOffset=<4 0 0>
     }]
   }
   n20 = "Constraint - dyn_armZipD_00_l_MSTR (PPPPPP)"!(
    ComponentPose=@n19
    Alpha=1
    ConstraintWeights_0=0.21
    ConstraintWeights_1:float=0.27
    ConstraintWeights_2:float=0.03
    ConstraintWeights_3:float=0.01
    ConstraintWeights_4:float=0.32
    ConstraintWeights_5:float=0.16
   ) -> (Pose) {
    alphaBoolBlend {blendOption=Linear}
    boneToModify {boneName=dyn_armZipD_00_l_MSTR}
    constraintSetup [
     {targetBone {boneName=upperarm_l}}
     {targetBone {boneName=lowerarm_twist_02_l}}
     {targetBone {boneName=lowerarm_twist_01_l}}
     {targetBone {boneName=upperarm_twist_02_l}}
     {targetBone {boneName=lowerarm_l}}
     {targetBone {boneName=deform_elbow_in_l}}
    ]
    constraintWeights [1,1,1,1,1,1]
   }
   n21 = Anim Dynamics - Bone: dyn_armZipD_01_l(ComponentPose=@n20, Alpha=1, ExternalForce="0,0,0") {
    angularDampingOverride=0.97
    angularSpringConstant=20000
    bAngularSpring=true
    bOverrideAngularDamping=true
    bUseGravityOverride=true
    simulationSpace=RootRelative
    alphaBoolBlend {blendOption=Linear}
    boundBone {boneName=dyn_armZipD_01_l}
    componentAppliedLinearAccClamp=<2000 2000 2000>
    componentLinearAccScale=<0.1 0.1 0.2>
    componentLinearVelScale=<0.2 0.2 0.4>
    gravityOverride {z=-980}
    physicsBodyDefinitions [{
      boundBone {boneName=dyn_armZipD_01_l}
      boxExtents=<10 20 20>
      constraintSetup {angularLimitsMax=<15 0 15>, angularLimitsMin=<-15 -45 -15>, angularTarget=<0 0 0>}
      localJointOffset=<4 0 0>
     }]
   }
   n22 = "Constraint - dyn_armZipA_00_r_MSTR (PPPPPP)"!(
    ComponentPose=@n21
    Alpha=1
    ConstraintWeights_0=0.43
    ConstraintWeights_1:float=0.08
    ConstraintWeights_2:float=0.07
    ConstraintWeights_3:float=0.24
    ConstraintWeights_4:float=0.14
    ConstraintWeights_5:float=0.04
   ) -> (Pose) {
    alphaBoolBlend {blendOption=Linear}
    boneToModify {boneName=dyn_armZipA_00_r_MSTR}
    constraintSetup [
     {targetBone {boneName=upperarm_twist_01_r}}
     {targetBone {boneName=upperarm_r}}
     {targetBone {boneName=clavicle_r}}
     {targetBone {boneName=upperarm_twist_02_r}}
     {targetBone {boneName=deform_upperarm_fr_r}}
     {targetBone {boneName=deform_upperarm_up_r}}
    ]
    constraintWeights [1,1,1,1,1,1]
   }
   n23 = Anim Dynamics - Bone: dyn_armZipA_01_r(ComponentPose=@n22, Alpha=1, ExternalForce="0,0,0") {
    angularDampingOverride=0.97
    angularSpringConstant=20000
    bAngularSpring=true
    bOverrideAngularDamping=true
    bUseGravityOverride=true
    simulationSpace=RootRelative
    alphaBoolBlend {blendOption=Linear}
    boundBone {boneName=dyn_armZipA_01_r}
    componentAppliedLinearAccClamp=<2000 2000 2000>
    componentLinearAccScale=<0.1 0.1 0.2>
    componentLinearVelScale=<0.2 0.2 0.4>
    gravityOverride {z=-980}
    physicsBodyDefinitions [{
      boundBone {boneName=dyn_armZipA_01_r}
      boxExtents=<10 20 20>
      constraintSetup {angularLimitsMax=<15 0 15>, angularLimitsMin=<-15 -45 -15>, angularTarget=<0 0 0>}
      localJointOffset=<4 0 0>
     }]
   }
   n24 = "Constraint - dyn_armZipB_00_r_MSTR (PPPPP)"!(
    ComponentPose=@n23
    Alpha=1
    ConstraintWeights_0=0.37
    ConstraintWeights_1:float=0.21
    ConstraintWeights_2:float=0.02
    ConstraintWeights_3:float=0.32
    ConstraintWeights_4:float=0.08
   ) -> (Pose) {
    alphaBoolBlend {blendOption=Linear}
    boneToModify {boneName=dyn_armZipB_00_r_MSTR}
    constraintSetup [
     {targetBone {boneName=upperarm_twist_01_r}}
     {targetBone {boneName=upperarm_r}}
     {targetBone {boneName=clavicle_r}}
     {targetBone {boneName=upperarm_twist_02_r}}
     {targetBone {boneName=deform_upperarm_fr_r}}
    ]
    constraintWeights [1,1,1,1,1]
   }
   n25 = Anim Dynamics - Bone: dyn_armZipB_01_r(ComponentPose=@n24, Alpha=1, ExternalForce="0,0,0") {
    angularDampingOverride=0.97
    angularSpringConstant=20000
    bAngularSpring=true
    bOverrideAngularDamping=true
    bUseGravityOverride=true
    simulationSpace=RootRelative
    alphaBoolBlend {blendOption=Linear}
    boundBone {boneName=dyn_armZipB_01_r}
    componentAppliedLinearAccClamp=<2000 2000 2000>
    componentLinearAccScale=<0.1 0.1 0.2>
    componentLinearVelScale=<0.2 0.2 0.4>
    gravityOverride {z=-980}
    physicsBodyDefinitions [{
      boundBone {boneName=dyn_armZipB_01_r}
      boxExtents=<10 20 20>
      constraintSetup {angularLimitsMax=<15 0 15>, angularLimitsMin=<-15 -45 -15>, angularTarget=<0 0 0>}
      localJointOffset=<4 0 0>
     }]
   }
   n26 = "Constraint - dyn_armZipC_00_r_MSTR (PPPP)"!(
    ComponentPose=@n25
    Alpha=1
    ConstraintWeights_0=0.36
    ConstraintWeights_1:float=0.32
    ConstraintWeights_2:float=0.05
    ConstraintWeights_3:float=0.27
   ) -> (Pose) {
    alphaBoolBlend {blendOption=Linear}
    boneToModify {boneName=dyn_armZipC_00_r_MSTR}
    constraintSetup [
     {targetBone {boneName=upperarm_twist_01_r}}
     {targetBone {boneName=upperarm_twist_02_r}}
     {targetBone {boneName=deform_upperarm_bk_r}}
     {targetBone {boneName=upperarm_r}}
    ]
    constraintWeights [1,1,1,1]
   }
   n27 = Anim Dynamics - Bone: dyn_armZipC_01_r(ComponentPose=@n26, Alpha=1, ExternalForce="0,0,0") {
    angularDampingOverride=0.97
    angularSpringConstant=20000
    bAngularSpring=true
    bOverrideAngularDamping=true
    bUseGravityOverride=true
    simulationSpace=RootRelative
    alphaBoolBlend {blendOption=Linear}
    boundBone {boneName=dyn_armZipC_01_r}
    componentAppliedLinearAccClamp=<2000 2000 2000>
    componentLinearAccScale=<0.1 0.1 0.2>
    componentLinearVelScale=<0.2 0.2 0.4>
    gravityOverride {z=-980}
    physicsBodyDefinitions [{
      boundBone {boneName=dyn_armZipC_01_r}
      boxExtents=<10 20 20>
      constraintSetup {angularLimitsMax=<15 0 15>, angularLimitsMin=<-15 -45 -15>, angularTarget=<0 0 0>}
      localJointOffset=<4 0 0>
     }]
   }
   n28 = "Constraint - dyn_armZipD_00_r_MSTR (PPPPPP)"!(
    ComponentPose=@n27
    Alpha=1
    ConstraintWeights_0=0.27
    ConstraintWeights_1:float=0.21
    ConstraintWeights_2:float=0.03
    ConstraintWeights_3:float=0.01
    ConstraintWeights_4:float=0.32
    ConstraintWeights_5:float=0.16
   ) -> (Pose) {
    alphaBoolBlend {blendOption=Linear}
    boneToModify {boneName=dyn_armZipD_00_r_MSTR}
    constraintSetup [
     {targetBone {boneName=lowerarm_twist_02_r}}
     {targetBone {boneName=upperarm_r}}
     {targetBone {boneName=lowerarm_twist_01_r}}
     {targetBone {boneName=upperarm_twist_02_r}}
     {targetBone {boneName=lowerarm_r}}
     {targetBone {boneName=deform_elbow_in_r}}
    ]
    constraintWeights [1,1,1,1,1,1]
   }
   n29 = Anim Dynamics - Bone: dyn_armZipD_01_r(ComponentPose=@n28, Alpha=1, ExternalForce="0,0,0") {
    angularDampingOverride=0.97
    angularSpringConstant=20000
    bAngularSpring=true
    bOverrideAngularDamping=true
    bUseGravityOverride=true
    simulationSpace=RootRelative
    alphaBoolBlend {blendOption=Linear}
    boundBone {boneName=dyn_armZipD_01_r}
    componentAppliedLinearAccClamp=<2000 2000 2000>
    componentLinearAccScale=<0.1 0.1 0.2>
    componentLinearVelScale=<0.2 0.2 0.4>
    gravityOverride {z=-980}
    physicsBodyDefinitions [{
      boundBone {boneName=dyn_armZipD_01_r}
      boxExtents=<10 20 20>
      constraintSetup {angularLimitsMax=<15 0 15>, angularLimitsMin=<-15 -45 -15>, angularTarget=<0 0 0>}
      localJointOffset=<4 0 0>
     }]
   }
   n30 = "Constraint - dyn_collar_00_c_MSTR (PPPPP)"!(
    ComponentPose=@n29
    Alpha=1
    ConstraintWeights_0=0.02
    ConstraintWeights_1:float=0.78
    ConstraintWeights_2:float=0.06
    ConstraintWeights_3:float=0.01
    ConstraintWeights_4:float=0.13
   ) -> (Pose) {
    alphaBoolBlend {blendOption=Linear}
    boneToModify {boneName=dyn_collar_00_c_MSTR}
    constraintSetup [
     {targetBone {boneName=spine_04}}
     {targetBone {boneName=spine_05}}
     {targetBone {boneName=neck_01}}
     {targetBone {boneName=head}}
     {targetBone {boneName=neck_02}}
    ]
    constraintWeights [1,1,1,1,1]
   }
   n31 = "Constraint - dyn_collar_00_l_MSTR (PPPPP)"!(
    ComponentPose=@n30
    Alpha=1
    ConstraintWeights_0=0.68
    ConstraintWeights_1:float=0.06
    ConstraintWeights_2:float=0.06
    ConstraintWeights_3:float=0.03
    ConstraintWeights_4:float=0.17
   ) -> (Pose) {
    alphaBoolBlend {blendOption=Linear}
    boneToModify {boneName=dyn_collar_00_l_MSTR}
    constraintSetup [
     {targetBone {boneName=spine_05}}
     {targetBone {boneName=clavicle_l}}
     {targetBone {boneName=neck_01}}
     {targetBone {boneName=head}}
     {targetBone {boneName=neck_02}}
    ]
    constraintWeights [1,1,1,1,1]
   }
   n32 = "Constraint - dyn_collar_00_r_MSTR (PPPPP)"!(
    ComponentPose=@n31
    Alpha=1
    ConstraintWeights_0=0.68
    ConstraintWeights_1:float=0.06
    ConstraintWeights_2:float=0.06
    ConstraintWeights_3:float=0.03
    ConstraintWeights_4:float=0.17
   ) -> (Pose) {
    alphaBoolBlend {blendOption=Linear}
    boneToModify {boneName=dyn_collar_00_r_MSTR}
    constraintSetup [
     {targetBone {boneName=spine_05}}
     {targetBone {boneName=clavicle_r}}
     {targetBone {boneName=neck_01}}
     {targetBone {boneName=head}}
     {targetBone {boneName=neck_02}}
    ]
    constraintWeights [1,1,1,1,1]
   }
   n33 = Get collarAlpha() -> (collarAlpha:double)
   n34 = "Transform (Modify) Bone - Bone: dyn_collar_01_l"(ComponentPose=@n32, Alpha=@n33, Translation="0,1,1", Rotation="75,0,0", Scale="1,1,1") {
    rotationMode=BMM_Additive
    rotationSpace=BCS_BoneSpace
    scaleMode=BMM_Additive
    scaleSpace=BCS_BoneSpace
    translationMode=BMM_Additive
    translationSpace=BCS_BoneSpace
    alphaBoolBlend {blendOption=Linear}
    alphaScaleBiasClamp {interpSpeedDecreasing=1}
    boneToModify {boneName=dyn_collar_01_l}
    rotation {pitch=75}
    translation {y=1, z=1}
   }
   n35 = "Transform (Modify) Bone - Bone: dyn_collar_01_c"(ComponentPose=@n34, Alpha=@n33, Translation="0,0,0", Rotation="15,30,0", Scale="1,1,1") {
    rotationMode=BMM_Additive
    rotationSpace=BCS_BoneSpace
    scaleMode=BMM_Additive
    scaleSpace=BCS_BoneSpace
    translationMode=BMM_Additive
    translationSpace=BCS_BoneSpace
    alphaBoolBlend {blendOption=Linear}
    alphaScaleBiasClamp {interpSpeedDecreasing=1}
    boneToModify {boneName=dyn_collar_01_c}
    rotation {pitch=15, yaw=30}
   }
   n36 = Get collarAlphaR() -> (collarAlphaR:double)
   n37 = "Transform (Modify) Bone - Bone: dyn_collar_01_r"(ComponentPose=@n35, Alpha=@n36, Translation="0,0,0", Rotation="75,0,0", Scale="1,1,1") {
    rotationMode=BMM_Additive
    rotationSpace=BCS_BoneSpace
    scaleMode=BMM_Additive
    scaleSpace=BCS_BoneSpace
    translationMode=BMM_Additive
    translationSpace=BCS_BoneSpace
    alphaBoolBlend {blendOption=Linear}
    alphaScaleBiasClamp {interpSpeedDecreasing=1}
    boneToModify {boneName=dyn_collar_01_r}
    rotation {pitch=75}
   }
   n38 = "Transform (Modify) Bone - Bone: dyn_collar_01_c"(ComponentPose=@n37, Alpha=@n36, Translation="0,0,0", Rotation="0,-30,0", Scale="1,1,1") {
    rotationMode=BMM_Additive
    rotationSpace=BCS_BoneSpace
    scaleMode=BMM_Additive
    scaleSpace=BCS_BoneSpace
    translationMode=BMM_Additive
    translationSpace=BCS_BoneSpace
    alphaBoolBlend {blendOption=Linear}
    alphaScaleBiasClamp {interpSpeedDecreasing=1}
    boneToModify {boneName=dyn_collar_01_c}
    rotation {yaw=-30}
   }
   n39 = "Transform (Modify) Bone - Bone: dyn_collar_01_l"(ComponentPose=@n38, Alpha=1, Translation="0,0,0", Rotation="15,0,0", Scale="1,1,1") {
    rotationMode=BMM_Additive
    rotationSpace=BCS_BoneSpace
    scaleMode=BMM_Additive
    scaleSpace=BCS_BoneSpace
    translationMode=BMM_Additive
    translationSpace=BCS_BoneSpace
    alphaBoolBlend {blendOption=Linear}
    alphaScaleBiasClamp {bInterpResult=true}
    boneToModify {boneName=dyn_collar_01_l}
    rotation {pitch=15}
   }
   n40 = "Transform (Modify) Bone - Bone: dyn_collar_01_r"(ComponentPose=@n39, Alpha=1, Translation="0,0,0", Rotation="15,0,0", Scale="1,1,1") {
    rotationMode=BMM_Additive
    rotationSpace=BCS_BoneSpace
    scaleMode=BMM_Additive
    scaleSpace=BCS_BoneSpace
    translationMode=BMM_Additive
    translationSpace=BCS_BoneSpace
    alphaBoolBlend {blendOption=Linear}
    alphaScaleBiasClamp {bInterpResult=true}
    boneToModify {boneName=dyn_collar_01_r}
    rotation {pitch=15}
   }
   n41 = "Transform (Modify) Bone - Bone: dyn_collar_01_c"(ComponentPose=@n40, Alpha=1, Translation="0,0,1", Rotation="15,0,0", Scale="1,1,1") {
    rotationMode=BMM_Additive
    rotationSpace=BCS_BoneSpace
    scaleMode=BMM_Additive
    scaleSpace=BCS_BoneSpace
    translationMode=BMM_Additive
    translationSpace=BCS_BoneSpace
    alphaBoolBlend {blendOption=Linear}
    alphaScaleBiasClamp {bInterpResult=true}
    boneToModify {boneName=dyn_collar_01_c}
    rotation {pitch=15}
    translation {z=1}
   }
   n42 = "Transform (Modify) Bone - Bone: dyn_collar_01_bk"(ComponentPose=@n41, Alpha=1, Translation="0,0,0", Rotation="90,0,0", Scale="1,1,1") {
    rotationMode=BMM_Additive
    rotationSpace=BCS_BoneSpace
    scaleMode=BMM_Additive
    scaleSpace=BCS_BoneSpace
    translationMode=BMM_Additive
    translationSpace=BCS_BoneSpace
    alphaBoolBlend {blendOption=Linear}
    boneToModify {boneName=dyn_collar_01_bk}
    rotation {pitch=90}
   }
   n43 = "Transform (Modify) Bone - Bone: dyn_collar_01_bkR"(ComponentPose=@n42, Alpha=1, Translation="0,0,0", Rotation="120,0,0", Scale="1,1,1") {
    rotationMode=BMM_Additive
    rotationSpace=BCS_BoneSpace
    scaleMode=BMM_Additive
    scaleSpace=BCS_BoneSpace
    translationMode=BMM_Additive
    translationSpace=BCS_BoneSpace
    alphaBoolBlend {blendOption=Linear}
    boneToModify {boneName=dyn_collar_01_bkR}
    rotation {pitch=120}
   }
   Output Pose(Result=Component To Local(ComponentPose=@n43)) {layerGroup=None}
  ]}
 EventGraph {nodes [
   Event BlueprintUpdateAnimation()
   n1 = GetOwnerSkeletalMeshComponent()
   n2 = BreakTransform(InTransform=GetDeltaTransformFromRefPose(self=@n1, BoneName=spine_05, BaseName=head))
   > n3 = Set neckDownRot(neckDownRot:double=MapRangeClamped(Value=BreakRotator(InRot=@n2.Rotation).Yaw, InRangeA=10, InRangeB=-5)) -> (Output_Get:double)
   n4 = BreakTransform(InTransform=GetDeltaTransformFromRefPose(self=@n1, BoneName=spine_05, BaseName=head))
   > n5 = Set neckSideRot(neckSideRot:double=MapRangeClamped(Value=BreakRotator(InRot=@n4.Rotation), InRangeA=0, InRangeB=-55)) -> (Output_Get:double)
   > Set collarAlpha(collarAlpha:double=float * float!(A=@n3, B=@n5, C:double=1) -> (ReturnValue) {is=title K2Node_PromotableOperator}) -> (Output_Get:double)
   n6 = BreakTransform(InTransform=GetDeltaTransformFromRefPose(self=@n1, BoneName=spine_05, BaseName=head))
   > n7 = Set neckRightRot(neckRightRot:double=MapRangeClamped(Value=BreakRotator(InRot=@n6.Rotation), InRangeA=0, InRangeB=40)) -> (Output_Get:double)
   > Set collarAlphaR(collarAlphaR:double=float * float(A=@n3, B=@n7, C=1)) -> (Output_Get:double)
   n8 = BreakTransform(InTransform=GetDeltaTransformFromRefPose(self=@n1, BoneName=spine_05, BaseName=head))
   > Set neckBackRot(neckBackRot:double=MapRangeClamped(Value=BreakRotator(InRot=@n8.Rotation).Yaw, InRangeA=30, InRangeB=90)) -> (Output_Get:double)
  ]}
]
variables [
 collarAlpha {type=number}
 collarAlphaR {type=number}
 neckBackRot {type=number}
 neckDownRot {type=number}
 neckRightRot {type=number}
 neckSideRot {type=number}
]
